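# Animation Blueprint for the Ballistic Shield on the Fortnite player skeleton, implementing Item Override animation layers.
# ItemFullBodyOverride is the working chain: a state machine adds stagger buildup, then feeds a charge aim offset and MaskLeftArm curve modifier.
# Equip/reload (per body type) and relax/landing state machines use per-bone blends, but their layers output the input pose, bypassing them.
# Other layers, AnimGraph and EventGraph are unwired or empty.
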AnimBlueprint /BallisticShieldGameplay/Assets/Anim/BallisticShield_AnimBP
target_skeleton=&Fortnite_M_Avg_Player_Skeleton
graphs [
 AnimGraph {nodes [Output Pose() {layerGroup=None}]}
 EventGraph {links [], nodes []}
 ItemAimOffset {nodes [
   Input Pose!() -> (Pose, AimOffsetAlpha:double, Yaw:double, Pitch:double) {name=InPose_AimOffset}
   Input Pose!() -> (Pose) {name=InPose_PreAimOffsets}
   Output Pose() {layerGroup=Item Override}
  ]}
 ItemFallAdditive {nodes [
   Input Pose() {name=InPoseItemFallAdditive}
   Output Pose() {layerGroup=Item Override}
  ]}
 ItemFallLandAdditive {nodes [
   Input Pose() {name=InPoseItemFallLandAdditive}
   Output Pose() {layerGroup=Item Override}
  ]}
 ItemFallLevel2Override {nodes [
   Input Pose() {name=InPose_FallLevel2}
   Output Pose() {layerGroup=Item Override}
  ]}
 ItemFinalPoseOverride {nodes [
   AnimGraphNode_StateMachine_0 = Equip / Reload_SM() {is=title AnimGraphNode_StateMachine}
   Blend Poses by bool(
    BlendPose_0=Use cached pose 'Final_Pose'()
    BlendPose_1=@AnimGraphNode_StateMachine_0
    BlendTime_0=0.2
    BlendTime_1=0.2
    bActiveValue=Get IsRush() -> (IsRush:bool)
   ) {blendType=Linear}
   Output Pose(Result=Input Pose() {name=InFinalPose}) {layerGroup=Item Override}
   Save cached pose 'Final_Pose'() {cacheName=Final_Pose}
  ]}
 Equip / Reload_SM {
  entry=Wait
  path=ItemFinalPoseOverride.AnimGraphNode_StateMachine_0.Equip / Reload_SM
  ids=id
  nodes [
   AnimStateNode_0 = Wait() {is=title AnimStateNode}
   AnimStateNode_1 = Reload() {is=title AnimStateNode}
   AnimStateNode_2 = Equip() {is=title AnimStateNode}
   AnimStateTransitionNode_0 = AnimStateTransitionNode() {crossfadeDuration=0.1}
   AnimStateTransitionNode_1 = AnimStateTransitionNode() {crossfadeDuration=0.1}
   AnimStateTransitionNode_2 = AnimStateTransitionNode() {bAutomaticRuleBasedOnSequencePlayerInState=true, crossfadeDuration=0.3}
   AnimStateTransitionNode_3 = AnimStateTransitionNode() {bAutomaticRuleBasedOnSequencePlayerInState=true, crossfadeDuration=0.3}
   AnimStateTransitionNode_4 = AnimStateTransitionNode()
   AnimStateTransitionNode_5 = AnimStateTransitionNode()
  ]
  links [
   {
    dst=Wait
    rule=ItemFinalPoseOverride.AnimGraphNode_StateMachine_0.Equip / Reload_SM.AnimStateTransitionNode_3.Transition
    src=Equip
    via=AnimStateTransitionNode_3
   }
   {
    dst=Wait
    rule=ItemFinalPoseOverride.AnimGraphNode_StateMachine_0.Equip / Reload_SM.AnimStateTransitionNode_4.Transition
    src=Equip
    via=AnimStateTransitionNode_4
   }
   {
    dst=Wait
    rule=ItemFinalPoseOverride.AnimGraphNode_StateMachine_0.Equip / Reload_SM.AnimStateTransitionNode_2.Transition
    src=Reload
    via=AnimStateTransitionNode_2
   }
   {
    dst=Wait
    rule=ItemFinalPoseOverride.AnimGraphNode_StateMachine_0.Equip / Reload_SM.AnimStateTransitionNode_5.Transition
    src=Reload
    via=AnimStateTransitionNode_5
   }
   {
    dst=Equip
    rule=ItemFinalPoseOverride.AnimGraphNode_StateMachine_0.Equip / Reload_SM.AnimStateTransitionNode_1.Transition
    src=Wait
    via=AnimStateTransitionNode_1
   }
   {
    dst=Reload
    rule=ItemFinalPoseOverride.AnimGraphNode_StateMachine_0.Equip / Reload_SM.AnimStateTransitionNode_0.Transition
    src=Wait
    via=AnimStateTransitionNode_0
   }
  ]
 }
 Wait {
  path=ItemFinalPoseOverride.AnimGraphNode_StateMachine_0.Equip / Reload_SM.AnimStateNode_0.Wait
  nodes [Output Animation Pose(Result=Use cached pose 'Final_Pose'())]
 }
 Reload {
  path=ItemFinalPoseOverride.AnimGraphNode_StateMachine_0.Equip / Reload_SM.AnimStateNode_1.Reload
  nodes [
   n1 = Blend Poses by int!(
    BlendPose_0:Pose Link Structure=Sequence Player 'BallisticShield_Reload_CMF'() {bLoopAnimation=false, sequence=&/BallisticShield_Reload_CMF}
    BlendPose_1:Pose Link Structure=Sequence Player 'BallisticShield_Reload_CMF'() {bLoopAnimation=false, sequence=&BallisticShield_Reload_CMF}
    BlendPose_2:Pose Link Structure=Sequence Player 'BallisticShield_Reload_CLM'() {bLoopAnimation=false, sequence=&/BallisticShield_Reload_CLM}
    BlendPose_3:Pose Link Structure=Sequence Player 'BallisticShield_Reload'() {bLoopAnimation=false, sequence=&/BallisticShield_Reload}
    BlendTime_0:float=0.1
    BlendTime_1:float=0.1
    BlendTime_2:float=0.1
    BlendTime_3:float=0.1
    ActiveChildIndex:int=Get GenderAndSize() -> (GenderAndSize:int=0)
   ) -> (Pose:Pose Link Structure) {
    is=title AnimGraphNode_BlendListByInt
    blendType=Linear
    blendPose [{linkId=-1, sourceLinkId=-1}, {linkId=-1, sourceLinkId=-1}, {linkId=-1, sourceLinkId=-1}, {linkId=-1, sourceLinkId=-1}]
    blendTime [0.1,0.1,0.1,0.1]
   }
   Output Animation Pose(Result=Layered blend per bone(BasePose=Use cached pose 'Final_Pose'(), BlendPoses_0=@n1, BlendWeights_0=Get ReloadUBAlpha() -> (ReloadUBAlpha:float)) {bMeshSpaceRotationBlend=true, layerSetup [{branchFilters [{blendDepth=3, boneName=Spine_01}]}]})
  ]
 }
 Equip {
  path=ItemFinalPoseOverride.AnimGraphNode_StateMachine_0.Equip / Reload_SM.AnimStateNode_2.Equip
  nodes [
   n1 = Blend Poses by int(
    BlendPose_0=Sequence Player 'BallisticShield_Equip_CMF'() {bLoopAnimation=false, sequence=/BallisticShieldGameplay/Assets/Anim/Equip/BallisticShield_Equip_CMF}
    BlendPose_1=Sequence Player 'BallisticShield_Equip_CLF'() {bLoopAnimation=false, sequence=/BallisticShieldGameplay/Assets/Anim/Equip/BallisticShield_Equip_CLF}
    BlendPose_2=Sequence Player 'BallisticShield_Equip_CLM'() {bLoopAnimation=false, sequence=/BallisticShieldGameplay/Assets/Anim/Equip/BallisticShield_Equip_CLM}
    BlendPose_3=Sequence Player 'BallisticShield_Equip'() {bLoopAnimation=false, sequence=/BallisticShieldGameplay/Assets/Anim/Equip/BallisticShield_Equip}
    BlendTime_0=0.1
    BlendTime_1=0.1
    BlendTime_2=0.1
    BlendTime_3=0.1
    ActiveChildIndex=Get GenderAndSize() -> (GenderAndSize=0)
   ) {
    blendType=Linear
    blendPose [{linkId=-1, sourceLinkId=-1}, {linkId=-1, sourceLinkId=-1}, {linkId=-1, sourceLinkId=-1}, {linkId=-1, sourceLinkId=-1}]
    blendTime [0.1,0.1,0.1,0.1]
   }
   Output Animation Pose(Result=Layered blend per bone(BasePose=Use cached pose 'Final_Pose'(), BlendPoses_0=@n1) {bMeshSpaceRotationBlend=true, layerSetup [{branchFilters [{blendDepth=3, boneName=Spine_01}]}]})
  ]
 }
 Transition {
  path=ItemFinalPoseOverride.AnimGraphNode_StateMachine_0.Equip / Reload_SM.AnimStateTransitionNode_0.Transition
  nodes [Result(bCanEnterTransition=Get IsReloading() -> (IsReloading:bool))]
 }
 Transition {
  path=ItemFinalPoseOverride.AnimGraphNode_StateMachine_0.Equip / Reload_SM.AnimStateTransitionNode_1.Transition
  nodes [Result(bCanEnterTransition=Get IsEquipping() -> (IsEquipping:bool))]
 }
 Transition {path=ItemFinalPoseOverride.AnimGraphNode_StateMachine_0.Equip / Reload_SM.AnimStateTransitionNode_2.Transition, nodes [Result()]}
 Transition {path=ItemFinalPoseOverride.AnimGraphNode_StateMachine_0.Equip / Reload_SM.AnimStateTransitionNode_3.Transition, nodes [Result()]}
 Transition {
  path=ItemFinalPoseOverride.AnimGraphNode_StateMachine_0.Equip / Reload_SM.AnimStateTransitionNode_4.Transition
  nodes [Result(bCanEnterTransition=NOT Boolean(A=Get IsEquipping()))]
 }
 Transition {
  path=ItemFinalPoseOverride.AnimGraphNode_StateMachine_0.Equip / Reload_SM.AnimStateTransitionNode_5.Transition
  nodes [Result(bCanEnterTransition=NOT Boolean(A=Get IsReloading()))]
 }
 ItemFullBodyAimOffsets {nodes [
   Input Pose!() -> (Pose, FullBody_AimOffsetAlpha:double, Aim_Yaw:double, Aim_Pitch:double) {name=InPoseFullBodyAimOffsets}
   Input Pose!() -> (Pose) {name=InPosePreFullBodyAimOffsets}
   Output Pose() {layerGroup=Item Override}
  ]}
 ItemFullBodyOverride {nodes [
   AnimGraphNode_StateMachine_1 = Stagger Buildup SM() {is=title AnimGraphNode_StateMachine}
   n1 = Get Yaw() -> (Yaw:float)
   n2 = Get ChargeAOAlpha() -> (ChargeAOAlpha:float)
   Output Pose(Result=Modify Curve(
     SourcePose=AimOffset Player 'BallisticShield_Charge_AO'!(X:float=@n1, Y:float=0, BasePose:Pose Link Structure=@AnimGraphNode_StateMachine_1, Alpha:float=@n2) -> (Pose:Pose Link Structure) {blendSpace=/BallisticShieldGameplay/Assets/Anim/Charge/BallisticShield_Charge_AO, alphaBoolBlend {blendOption=Linear}}
     CurveValues_0=1
     Alpha=Get MaskLeftArmAlpha() -> (MaskLeftArmAlpha:float)
    ) {curveNames [MaskLeftArm], curveValues [0]}) {layerGroup=Item Override}
   Save cached pose 'FullBody_Pose'(Pose=Input Pose() {name=InPoseFullBody}) {cacheName=FullBody_Pose}
  ]}
 Stagger Buildup SM {
  entry=Wait
  path=ItemFullBodyOverride.AnimGraphNode_StateMachine_1.Stagger Buildup SM
  ids=id
  nodes [
   To Stagger Buildup() {is=title AnimStateAliasNode, bGlobalAlias=true}
   To Wait() {is=title AnimStateAliasNode, bGlobalAlias=true}
   AnimStateNode_0 = Wait()
   AnimStateNode_1 = Stagger Buildup() {is=title AnimStateNode}
   AnimStateTransitionNode_1 = AnimStateTransitionNode() {crossfadeDuration=0.05}
   AnimStateTransitionNode_2 = AnimStateTransitionNode() {bAutomaticRuleBasedOnSequencePlayerInState=true, crossfadeDuration=0.05}
  ]
  links [
   {
    dst=Stagger Buildup
    rule=ItemFullBodyOverride.AnimGraphNode_StateMachine_1.Stagger Buildup SM.AnimStateTransitionNode_1.Transition
    src=To Stagger Buildup
    via=AnimStateTransitionNode_1
   }
   {
    dst=Wait
    rule=ItemFullBodyOverride.AnimGraphNode_StateMachine_1.Stagger Buildup SM.AnimStateTransitionNode_2.Transition
    src=To Wait
    via=AnimStateTransitionNode_2
   }
  ]
 }
 Wait {
  path=ItemFullBodyOverride.AnimGraphNode_StateMachine_1.Stagger Buildup SM.AnimStateNode_0.Wait
  nodes [Output Animation Pose(Result=Use cached pose 'FullBody_Pose'())]
 }
 Stagger Buildup {
  path=ItemFullBodyOverride.AnimGraphNode_StateMachine_1.Stagger Buildup SM.AnimStateNode_1.Stagger Buildup
  nodes [
   n1 = Get StaggerAmount() -> (StaggerAmount:float)
   Output Animation Pose(Result=Apply Additive(
     Base=Use cached pose 'FullBody_Pose'()
     Additive=Blendspace Player 'BallisticShield_StaggerBuildup_BS'(X=@n1, Y=0) {bLoop=false, blendSpace=&/BallisticShield_StaggerBuildup_BS}
    ) {alphaBoolBlend {blendOption=Linear}})
  ]
 }
 Transition {
  path=ItemFullBodyOverride.AnimGraphNode_StateMachine_1.Stagger Buildup SM.AnimStateTransitionNode_1.Transition
  nodes [Result(bCanEnterTransition=Get bIsStaggerBuildup() -> (bIsStaggerBuildup:bool))]
 }
 Transition {path=ItemFullBodyOverride.AnimGraphNode_StateMachine_1.Stagger Buildup SM.AnimStateTransitionNode_2.Transition, nodes [Result()]}
 ItemIdleOverride {nodes [
   Input Pose() {name=InPose_Idle}
   Output Pose() {layerGroup=Item Override}
  ]}
 ItemJetPackJumpAdditive {nodes [Output Pose() {layerGroup=Item Override}]}
 ItemJetPackStartAdditive {nodes [Output Pose() {layerGroup=Item Override}]}
 ItemJumpLoopAdditive {nodes [
   Input Pose() {name=InPoseItemJumpLoopAdditive}
   Output Pose() {layerGroup=Item Override}
  ]}
 ItemJumpOverride {nodes [
   Input Pose() {name=InPoseJump}
   Output Pose() {layerGroup=Item Override}
  ]}
 ItemJumpUpAdditive {nodes [
   Input Pose() {name=InPoseItemJumpUpAdditive}
   Output Pose() {layerGroup=Item Override}
  ]}
 ItemPreAimOffsets {nodes [
   Input Pose() {name=InPosePreAimOffsets}
   Output Pose() {layerGroup=Item Override}
  ]}
 ItemPreIK {nodes [
   Input Pose() {name=InPose_PreIK}
   Output Pose() {layerGroup=Item Override}
  ]}
 ItemRightHandIK {nodes [
   Input Pose() {name=InPosePreRightHandIK}
   Input Pose() {name=InPosePostRightHandIK}
   Output Pose() {layerGroup=Item Override}
  ]}
 ItemSkeletalControl {nodes [
   Input Pose()
   Output Pose() {layerGroup=Item Override}
  ]}
 ItemSwimJumpFallLoop {nodes [Output Pose() {layerGroup=Item Override}]}
 ItemSwimJumpSurfaceEnd {nodes [Output Pose() {layerGroup=Item Override}]}
 ItemSwimJumpSurfaceLoop {nodes [Output Pose() {layerGroup=Item Override}]}
 ItemUpperBody {nodes [
   Output Pose(Result=Input Pose!() -> (Pose, InputParam:Fort Anim Input Adjusted Aim Structure) {name=InPoseUpperBody}) {layerGroup=Item Override}
   Save cached pose 'UpperBody Pose'() {cacheName=UpperBody Pose}
   AnimGraphNode_StateMachine_0 = Relax Transition UB SM() {is=title AnimGraphNode_StateMachine}
  ]}
 Relax Transition UB SM {
  entry=Wait UB
  path=ItemUpperBody.AnimGraphNode_StateMachine_0.Relax Transition UB SM
  ids=id
  nodes [
   To Transition to Relaxed UB() {is=title AnimStateAliasNode, bGlobalAlias=true}
   To Wait UB() {is=title AnimStateAliasNode, bGlobalAlias=true}
   AnimStateNode_0 = Wait UB() {is=title AnimStateNode}
   AnimStateNode_1 = Transition to Relaxed UB() {is=title AnimStateNode}
   AnimStateTransitionNode_0 = AnimStateTransitionNode() {bAutomaticRuleBasedOnSequencePlayerInState=true, blendMode=Linear, crossfadeDuration=0.35, priorityOrder=2}
   AnimStateTransitionNode_1 = AnimStateTransitionNode()
   AnimStateTransitionNode_2 = AnimStateTransitionNode() {blendMode=Linear, crossfadeDuration=0.25, priorityOrder=2}
   AnimStateTransitionNode_3 = AnimStateTransitionNode()
   AnimStateTransitionNode_4 = AnimStateTransitionNode()
  ]
  links [
   {
    dst=Transition to Relaxed UB
    rule=ItemUpperBody.AnimGraphNode_StateMachine_0.Relax Transition UB SM.AnimStateTransitionNode_2.Transition
    src=To Transition to Relaxed UB
    via=AnimStateTransitionNode_2
   }
   {
    dst=Wait UB
    rule=ItemUpperBody.AnimGraphNode_StateMachine_0.Relax Transition UB SM.AnimStateTransitionNode_0.Transition
    src=To Wait UB
    via=AnimStateTransitionNode_0
   }
   {
    dst=Wait UB
    rule=ItemUpperBody.AnimGraphNode_StateMachine_0.Relax Transition UB SM.AnimStateTransitionNode_1.Transition
    src=To Wait UB
    via=AnimStateTransitionNode_1
   }
   {
    dst=Wait UB
    rule=ItemUpperBody.AnimGraphNode_StateMachine_0.Relax Transition UB SM.AnimStateTransitionNode_3.Transition
    src=To Wait UB
    via=AnimStateTransitionNode_3
   }
   {
    dst=Wait UB
    rule=ItemUpperBody.AnimGraphNode_StateMachine_0.Relax Transition UB SM.AnimStateTransitionNode_4.Transition
    src=To Wait UB
    via=AnimStateTransitionNode_4
   }
  ]
 }
 Wait UB {
  path=ItemUpperBody.AnimGraphNode_StateMachine_0.Relax Transition UB SM.AnimStateNode_0.Wait UB
  nodes [
   AnimGraphNode_StateMachine_0 = Jump Landing Relax UB SM() {is=title AnimGraphNode_StateMachine}
   Output Animation Pose(Result=@AnimGraphNode_StateMachine_0)
  ]
 }
 Jump Landing Relax UB SM {
  entry=Wait UB Land
  path=ItemUpperBody.AnimGraphNode_StateMachine_0.Relax Transition UB SM.AnimStateNode_0.Wait UB.AnimGraphNode_StateMachine_0.Jump Landing Relax UB SM
  ids=id
  nodes [
   AnimStateNode_0 = Wait UB Land() {is=title AnimStateNode}
   AnimStateNode_1 = Land Relax UB() {is=title AnimStateNode}
   AnimStateTransitionNode_5 = AnimStateTransitionNode()
   AnimStateTransitionNode_6 = AnimStateTransitionNode() {crossfadeDuration=0.4}
   AnimStateTransitionNode_7 = AnimStateTransitionNode() {crossfadeDuration=0}
  ]
  links [
   {
    dst=Wait UB Land
    rule=ItemUpperBody.AnimGraphNode_StateMachine_0.Relax Transition UB SM.AnimStateNode_0.Wait UB.AnimGraphNode_StateMachine_0.Jump Landing Relax UB SM.AnimStateTransitionNode_6.Transition
    src=Land Relax UB
    via=AnimStateTransitionNode_6
   }
   {
    dst=Wait UB Land
    rule=ItemUpperBody.AnimGraphNode_StateMachine_0.Relax Transition UB SM.AnimStateNode_0.Wait UB.AnimGraphNode_StateMachine_0.Jump Landing Relax UB SM.AnimStateTransitionNode_7.Transition
    src=Land Relax UB
    via=AnimStateTransitionNode_7
   }
   {
    dst=Land Relax UB
    rule=ItemUpperBody.AnimGraphNode_StateMachine_0.Relax Transition UB SM.AnimStateNode_0.Wait UB.AnimGraphNode_StateMachine_0.Jump Landing Relax UB SM.AnimStateTransitionNode_5.Transition
    src=Wait UB Land
    via=AnimStateTransitionNode_5
   }
  ]
 }
 Wait UB Land {
  path=ItemUpperBody.AnimGraphNode_StateMachine_0.Relax Transition UB SM.AnimStateNode_0.Wait UB.AnimGraphNode_StateMachine_0.Jump Landing Relax UB SM.AnimStateNode_0.Wait UB Land
  nodes [Output Animation Pose(Result=Use cached pose 'UpperBody Pose'())]
 }
 Land Relax UB {
  path=ItemUpperBody.AnimGraphNode_StateMachine_0.Relax Transition UB SM.AnimStateNode_0.Wait UB.AnimGraphNode_StateMachine_0.Jump Landing Relax UB SM.AnimStateNode_1.Land Relax UB
  nodes [Output Animation Pose(Result=Layered blend per bone(
     BasePose=Use cached pose 'UpperBody Pose'()
     BlendPoses_0=Sequence Player 'BallisticShield_Relaxed_Pose'() {bLoopAnimation=false, sequence=&/BallisticShield_Relaxed_Pose}
    ) {bMeshSpaceRotationBlend=true, layerSetup [{branchFilters [{boneName=clavicle_r}]}]})]
 }
 Transition {
  path=ItemUpperBody.AnimGraphNode_StateMachine_0.Relax Transition UB SM.AnimStateNode_0.Wait UB.AnimGraphNode_StateMachine_0.Jump Landing Relax UB SM.AnimStateTransitionNode_5.Transition
  nodes [Result(bCanEnterTransition=Get!() -> (bEnterLanding:bool, Enter Landing_b:bool).Enter Landing_b)]
 }
 Transition {
  path=ItemUpperBody.AnimGraphNode_StateMachine_0.Relax Transition UB SM.AnimStateNode_0.Wait UB.AnimGraphNode_StateMachine_0.Jump Landing Relax UB SM.AnimStateTransitionNode_6.Transition
  nodes [Result(bCanEnterTransition=Get!() -> (bExitLanding:bool, Exit Landing_b:bool).Exit Landing_b)]
 }
 Transition {
  path=ItemUpperBody.AnimGraphNode_StateMachine_0.Relax Transition UB SM.AnimStateNode_0.Wait UB.AnimGraphNode_StateMachine_0.Jump Landing Relax UB SM.AnimStateTransitionNode_7.Transition
  nodes [Result(bCanEnterTransition=NOT Boolean(A=Get bIsRelaxedLevel1() -> (bIsRelaxedLevel1:bool)))]
 }
 Transition to Relaxed UB {
  path=ItemUpperBody.AnimGraphNode_StateMachine_0.Relax Transition UB SM.AnimStateNode_1.Transition to Relaxed UB
  nodes [
   n1 = Layered blend per bone(
    BasePose=Use cached pose 'UpperBody Pose'()
    BlendPoses_0=Sequence Player 'BallisticShield_TransitionToRelaxed'() {bLoopAnimation=false, sequence=&/BallisticShield_TransitionToRelaxed}
   ) {
    bMeshSpaceRotationBlend=true
    layerSetup [{branchFilters [{boneName=Spine_01}, {blendDepth=-1, boneName=clavicle_l}, {blendDepth=-1, boneName=neck_01}]}]
   }
   Output Animation Pose(Result=@n1)
  ]
 }
 Transition {path=ItemUpperBody.AnimGraphNode_StateMachine_0.Relax Transition UB SM.AnimStateTransitionNode_0.Transition, nodes [Result()]}
 Transition {
  path=ItemUpperBody.AnimGraphNode_StateMachine_0.Relax Transition UB SM.AnimStateTransitionNode_1.Transition
  nodes [Result(bCanEnterTransition=NOT Boolean(A=Get bIsRelaxedLevel1()))]
 }
 Transition {
  path=ItemUpperBody.AnimGraphNode_StateMachine_0.Relax Transition UB SM.AnimStateTransitionNode_2.Transition
  nodes [Result(bCanEnterTransition=Get IsGoingRelaxed() -> (IsGoingRelaxed:bool))]
 }
 Transition {
  path=ItemUpperBody.AnimGraphNode_StateMachine_0.Relax Transition UB SM.AnimStateTransitionNode_3.Transition
  nodes [Result(bCanEnterTransition=Get IsFalling() -> (IsFalling:bool))]
 }
 Transition {
  path=ItemUpperBody.AnimGraphNode_StateMachine_0.Relax Transition UB SM.AnimStateTransitionNode_4.Transition
  nodes [Result(bCanEnterTransition=Get!() -> (bEnterLanding:bool, Enter Landing_b:bool).Enter Landing_b)]
 }
 ItemZipLineStartAdditive {nodes [Output Pose() {layerGroup=Item Override}]}
]
variables [
 IsFalling {type=boolean}
 IsGoingRelaxed {type=boolean}
]
