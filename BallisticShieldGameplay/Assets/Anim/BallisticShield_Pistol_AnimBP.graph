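# Animation Blueprint for the Ballistic Shield's pistol skeleton, playing the weapon-side equip and reload animations.
# Each update casts the owning actor to B_Paprika_BallisticShield_Weap and copies its equipping and reloading states into booleans; IsFiring is declared but unused.
# These drive a Wait (ref pose), Equip and Reloading state machine with instant transitions, feeding DefaultSlot.
# A global alias lets any state enter Reloading, which returns to Wait automatically when its sequence ends.
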AnimBlueprint /BallisticShieldGameplay/Assets/Anim/BallisticShield_Pistol_AnimBP
target_skeleton=/BallisticShieldGameplay/Assets/Meshes/Shield_Ballistic_Pistol_Skeleton
graphs [
 AnimGraph {nodes [
   AnimGraphNode_StateMachine_0 = Pistol_SM() {is=title AnimGraphNode_StateMachine}
   Output Pose(Result=Slot 'DefaultSlot'(Source=@AnimGraphNode_StateMachine_0)) {layerGroup=None}
  ]}
 Pistol_SM {
  entry=Wait
  path=AnimGraph.AnimGraphNode_StateMachine_0.Pistol_SM
  ids=id
  nodes [
   To Reloading() {is=title AnimStateAliasNode, bGlobalAlias=true}
   To Wait() {is=title AnimStateAliasNode, also {aliases [Reloading]}}
   AnimStateNode_0 = Wait() {is=title AnimStateNode}
   AnimStateNode_1 = Equip() {is=title AnimStateNode}
   AnimStateNode_2 = Reloading() {is=title AnimStateNode}
   AnimStateTransitionNode_0 = AnimStateTransitionNode() {crossfadeDuration=0}
   AnimStateTransitionNode_1 = AnimStateTransitionNode() {crossfadeDuration=0}
   AnimStateTransitionNode_2 = AnimStateTransitionNode() {bAutomaticRuleBasedOnSequencePlayerInState=true, crossfadeDuration=0, priorityOrder=10}
   AnimStateTransitionNode_3 = AnimStateTransitionNode() {crossfadeDuration=0}
  ]
  links [
   {
    dst=Wait
    rule=AnimGraph.AnimGraphNode_StateMachine_0.Pistol_SM.AnimStateTransitionNode_1.AnimationTransitionGraph_0
    src=Equip
    via=AnimStateTransitionNode_1
   }
   {
    dst=Reloading
    rule=AnimGraph.AnimGraphNode_StateMachine_0.Pistol_SM.AnimStateTransitionNode_3.AnimationTransitionGraph_1
    src=To Reloading
    via=AnimStateTransitionNode_3
   }
   {
    dst=Wait
    rule=AnimGraph.AnimGraphNode_StateMachine_0.Pistol_SM.AnimStateTransitionNode_2.AnimationTransitionGraph_1
    src=To Wait
    via=AnimStateTransitionNode_2
   }
   {
    dst=Equip
    rule=AnimGraph.AnimGraphNode_StateMachine_0.Pistol_SM.AnimStateTransitionNode_0.AnimationTransitionGraph_0
    src=Wait
    via=AnimStateTransitionNode_0
   }
  ]
 }
 Wait {path=AnimGraph.AnimGraphNode_StateMachine_0.Pistol_SM.AnimStateNode_0.Wait, nodes [Output Animation Pose(Result=Local Space Ref Pose())]}
 Equip {
  path=AnimGraph.AnimGraphNode_StateMachine_0.Pistol_SM.AnimStateNode_1.Equip
  nodes [Output Animation Pose(Result=Sequence Player 'BallisticShield_Equip_WEP_R'() {bLoopAnimation=false, sequence=/BallisticShieldGameplay/Assets/Anim/Equip/BallisticShield_Equip_WEP_R})]
 }
 Reloading {
  path=AnimGraph.AnimGraphNode_StateMachine_0.Pistol_SM.AnimStateNode_2.Reloading
  nodes [Output Animation Pose(Result=Sequence Player 'BallisticShield_Reload_WEP1'() {bLoopAnimation=false, sequence=&/BallisticShield_Reload_WEP1})]
 }
 AnimationTransitionGraph_0 {
  path=AnimGraph.AnimGraphNode_StateMachine_0.Pistol_SM.AnimStateTransitionNode_0.AnimationTransitionGraph_0
  nodes [Result(bCanEnterTransition=Get IsEquipping() -> (IsEquipping:bool))]
 }
 AnimationTransitionGraph_0 {
  path=AnimGraph.AnimGraphNode_StateMachine_0.Pistol_SM.AnimStateTransitionNode_1.AnimationTransitionGraph_0
  nodes [Result(bCanEnterTransition=NOT Boolean(A=Get IsEquipping()))]
 }
 AnimationTransitionGraph_1 {path=AnimGraph.AnimGraphNode_StateMachine_0.Pistol_SM.AnimStateTransitionNode_2.AnimationTransitionGraph_1, nodes [Result()]}
 AnimationTransitionGraph_1 {
  path=AnimGraph.AnimGraphNode_StateMachine_0.Pistol_SM.AnimStateTransitionNode_3.AnimationTransitionGraph_1
  nodes [Result(bCanEnterTransition=Get IsReloading() -> (IsReloading:bool))]
 }
 EventGraph {nodes [
   Event BlueprintUpdateAnimation()
   n1 = GetOwningActor()
   > Is Valid(InputObject=@n1)
   > n2 = Cast To B_Paprika_BallisticShield_Weap!(execute, Object=@n1) -> (then, CastFailed, AsB Paprika Ballistic Shield Weap:B Paprika Ballistic Shield Weap Object Reference) {is=title K2Node_DynamicCast}
   > Set IsEquipping(IsEquipping:bool=IsBeingEquipped!(self:Fort Weapon Object Reference=@n2) -> (ReturnValue:bool)) -> (Output_Get:bool)
   > Set IsReloading(IsReloading:bool=IsReloading!(self:Fort Weapon Object Reference=@n2) -> (ReturnValue:bool)) -> (Output_Get:bool)
  ]}
]
variables [
 IsEquipping {type=boolean}
 IsFiring {type=boolean}
 IsReloading {type=boolean}
]
