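; Camera lens effect Blueprint for the ballistic shield stagger, built on a Niagara actor.
; On BeginPlay it copies the actor's initial life span into a float parameter named "CameraEffectLifteime" on its Niagara component.
; This ties the particle effect's timing to the actor's lifetime; the construction script is empty.
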(event EventBeginPlay
  (Effects|Components|ParticleSystem|SetFloatParameter (Variables|NiagaraActor|GetNiagaraComponent) "CameraEffectLifteime" (Variables|Actor|GetInitialLifeSpan)))

(fn ConstructionScript)
