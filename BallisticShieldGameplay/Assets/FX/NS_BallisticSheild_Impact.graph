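# Niagara system for ballistic shield impacts, bursting short-lived ricochet spark sprites.
# Two active emitters, Rico_bullets and a denser Small_rico, fire randomly rolled velocity cones, slowed by drag and gravity and stretched by speed.
# System.Impactintencity lerps spawn color, but an update color curve then sets a randomly brightened white-to-orange fade.
# Two variant emitters are disabled, and User.ImpactStage is exposed as a float input.
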NiagaraSystem /BallisticShieldGameplay/Assets/FX/NS_BallisticSheild_Impact
dependencies {
 usedDataInterfaces [&NiagaraDataInterfaceCurve, &NiagaraDataInterfaceColorCurve, &NiagaraDataInterfaceArrayFloat]
 usedDynamicInputs [&SelectFloatFromArray]
 usedModules [
  &self:SystemSpawnScript.SystemScriptSource.SystemScriptGraph.NiagaraNodeAssignment_1.SetVariables_#1
  &self:SystemSpawnScript.SystemScriptSource.SystemScriptGraph.NiagaraNodeAssignment_0.SetVariables_#2
  &SystemState
 ]
}
emitters {
 Rico_bullets {
  emitterUpdateScript [
   &EmitterState()
   &SpawnBurst_Instantaneous(Spawn Count=3, Use Spawn Probability=false)
  ]
  particleSpawnScript [
   &InitializeParticle(
    Color=&LerpLinearColors(StartColor=<rgba 2 0.575 0.1 1>, EndColor=<rgba 5 3.1239 0.7531 1>, LerpFactor=$System.Impactintencity)
    Lifetime Max=0.25
    Lifetime Min=0.2
    Lifetime Mode=::NewEnumerator1
    Sprite Size Mode=::NewEnumerator3
    Uniform Sprite Size=4
    UsePositionOffset=false
    Write Material Random=true
   ) {module=InitializeParticle001}
   &AddVelocityInCone(
    Cone Angle=30
    Cone Axis=/Niagara/DynamicInputs/Vectors/RotateVector(
     Vector To Rotate:Vector3f=<0.35 0 1>
     Yaw:NiagaraFloat=0
     Pitch:NiagaraFloat=0
     Roll:NiagaraFloat=&Multiply_Float(A=35, B=&RandomRangeFloat(Minimum=-1, Maximum=1))
     Rotation Rate:NiagaraFloat=1
    )
    Module.Use Velocity Falloff On Cone Axis=false
    Velocity Strength=&Multiply_Float(A=&UniformRangedFloat(Minimum=300, Maximum=200), B=10)
   )
  ]
  particleUpdateScript [
   &ParticleState()
   &ScaleColor(
    Scale Alpha=&FloatFromCurve(FloatCurve=DataInterface() {propertyValues {Curve {keys [
         {tangentMode=RCTM_User, time=0, value=1}
         {arriveTangent=-0.0405, interpMode=RCIM_Cubic, leaveTangent=-0.0405, tangentMode=RCTM_User, time=0.7009, value=1}
         {arriveTangent=-8.0089, interpMode=RCIM_Cubic, leaveTangent=-8.0089, tangentMode=RCTM_User, time=1, value=0}
        ]}}})
    ScaleA=true
    ScaleRGB=true
   )
   &Drag(Drag=&RandomRangeFloat(Minimum=20, Maximum=25), Use Linear Drag=true, Use Rotational Drag=true)
   &GravityForce(Gravity=<0 0 -1960>)
   &SolveForcesAndVelocity(Clamp Velocity=false, Limit Acceleration=false)
   &ScaleSpriteSizeBySpeed(Max Scale Factor=&RandomRangeVector2D(Minimum=<0.5 40>, Maximum=<0.5 35>), Min Scale Factor=<0.35 0.1>)
   &DynamicMaterialParameters(
    Index 0 Param 1=1
    Index 0 Param 2=1
    Index 0 Param 3=&FloatFromCurve(FloatCurve=DataInterface() {propertyValues {Curve {keys [<at 0 1 RCIM_Cubic>,<at 1 0 RCIM_Cubic>]}}})
    Index 0 Param 4=1
    Module.Index0Param0Write=true
    Module.Index0Param1Write=true
    Module.Index0Param2Write=true
    Module.Index0Param3Write=true
   )
   &Color(Color=&MultiplyLinearColor(
     Linear Color=&ColorFromCurve(ColorCurve=DataInterface() {propertyValues {
        AlphaCurve {keys [<at 0.8243 1 RCIM_Cubic>,<at 1 0 RCIM_Cubic>]}
        BlueCurve {keys [
          <at 0.0115 1.25 RCIM_Cubic>
          {arriveTangentWeight=0.209, interpMode=RCIM_Cubic, leaveTangentWeight=0.1865, tangentMode=RCTM_User, time=0.1937, value=0.2}
          {arriveTangent=-0.2642, interpMode=RCIM_Cubic, leaveTangent=-0.2642, time=0.6501, value=0}
          <at 0.9508 0 RCIM_Cubic>
         ]}
        GreenCurve {keys [
          <at 0.0115 1.25 RCIM_Cubic>
          {
           arriveTangent=-1.5659
           arriveTangentWeight=0.1735
           interpMode=RCIM_Cubic
           leaveTangent=-1.5659
           leaveTangentWeight=0.1548
           time=0.1937
           value=0.6
          }
          {arriveTangent=-0.7925, interpMode=RCIM_Cubic, leaveTangent=-0.7925, time=0.6501, value=0.25}
          <at 0.9508 0 RCIM_Cubic>
         ]}
        RedCurve {keys [
          <at 0.0115 1.25 RCIM_Cubic>
          {
           arriveTangent=-0.3915
           arriveTangentWeight=0.1125
           interpMode=RCIM_Cubic
           leaveTangent=-0.3915
           leaveTangentWeight=0.1004
           time=0.1937
           value=1
          }
          <at 0.6501 1 RCIM_Cubic>
          <at 0.9508 1 RCIM_Cubic>
         ]}
       }})
     Scale Factor=&MakeFloat4FromVectorAndFloat(
      "Vector (RGB)"=&VectorFromFloat(Value=&RandomRangeFloat(Minimum=0.5, Maximum=2))
      "Float (Alpha)"=1
     )
    ))
  ]
  rendererClasses [&NiagaraSpriteRendererProperties]
  renderers [{rendererClass=&NiagaraSpriteRendererProperties}]
 }
 Rico_bullets002 {
  bEnabled=false
  emitterUpdateScript [
   &EmitterState()
   &SpawnBurst_Instantaneous(Spawn Count=3, Use Spawn Probability=false)
  ]
  particleSpawnScript [
   &InitializeParticle(
    Color=&LerpLinearColors(StartColor=<rgba 2 0.575 0.1 1>, EndColor=<rgba 5 3.1239 0.7531 1>, LerpFactor=$System.Impactintencity)
    Lifetime Max=0.25
    Lifetime Min=0.2
    Lifetime Mode=::NewEnumerator1
    Sprite Size=&Lerp_Vector2(
     A=&RandomRangeVector2D(Minimum=<3 5>, Maximum=<2 10>)
     B=&RandomRangeVector2D(Minimum=<2 15>, Maximum=<4 10>)
     Alpha=$System.Impactintencity
    )
    Sprite Size Mode=::NewEnumerator4
    UsePositionOffset=false
    Write Material Random=true
   ) {module=InitializeParticle001}
   &AddVelocityInCone(
    Cone Angle=30
    Cone Axis=&RotateVector(
     Vector To Rotate:Vector3f=<0.35 0 1>
     Yaw:NiagaraFloat=0
     Pitch:NiagaraFloat=0
     Roll:NiagaraFloat=&Multiply_Float(A=35, B=&RandomRangeFloat(Minimum=-1, Maximum=1))
     Rotation Rate:NiagaraFloat=1
    )
    Module.Use Velocity Falloff On Cone Axis=false
    Velocity Strength=&Multiply_Float(A=&UniformRangedFloat(Minimum=300, Maximum=200), B=10)
   )
  ]
  particleUpdateScript [
   &ParticleState()
   &ScaleColor(
    Scale Alpha=&FloatFromCurve(FloatCurve=DataInterface() {propertyValues {Curve {keys [
         {tangentMode=RCTM_User, time=0, value=1}
         {arriveTangent=-0.0405, interpMode=RCIM_Cubic, leaveTangent=-0.0405, tangentMode=RCTM_User, time=0.7009, value=1}
         {arriveTangent=-8.0089, interpMode=RCIM_Cubic, leaveTangent=-8.0089, tangentMode=RCTM_User, time=1, value=0}
        ]}}})
    ScaleA=true
    ScaleRGB=true
   )
   &Drag(
    Drag=&Lerp_Float(
     A=&Multiply_Float(A=&RandomRangeFloat(Minimum=10, Maximum=12), B=2.2)
     B=&Multiply_Float(A=&RandomRangeFloat(Minimum=10, Maximum=12), B=2.2)
     Alpha=0.5
    )
    Use Linear Drag=true
    Use Rotational Drag=false
   )
   &SolveForcesAndVelocity(Clamp Velocity=false, Limit Acceleration=false)
   &ScaleSpriteSize(Scale Factor=&MakeVector2D(
     X=&FloatFromCurve(FloatCurve=DataInterface() {propertyValues {Curve {keys [{arriveTangent=-0.0405, interpMode=RCIM_Cubic, time=0, value=1}, {interpMode=RCIM_Cubic, leaveTangent=-8.0089, time=1, value=0.75}]}}})
     Y=&FloatFromCurve(FloatCurve=DataInterface() {propertyValues {Curve {keys [{arriveTangent=-2.7038, interpMode=RCIM_Cubic, time=0, value=1}, {interpMode=RCIM_Cubic, leaveTangent=-8.0089, time=1, value=0}]}}})
    ))
   &DynamicMaterialParameters(
    Index 0 Param 1=1
    Index 0 Param 2=1
    Index 0 Param 3=&FloatFromCurve(FloatCurve=DataInterface() {propertyValues {Curve {keys [{interpMode=RCIM_Cubic, tangentMode=RCTM_User, time=0, value=1}, {interpMode=RCIM_Cubic, tangentMode=RCTM_User, time=1, value=0}]}}})
    Index 0 Param 4=1
    Module.Param0WriteEnabled=true
    Module.Param1WriteEnabled=true
    Module.Param2WriteEnabled=true
    Module.Param3WriteEnabled=true
   ) {module=DynamicMaterialParameters001}
   &ScaleSpriteSizeBySpeed(Max Scale Factor=<0.8 6>, Min Scale Factor=<1 3>, Velocity Threshold=2000)
  ]
  rendererClasses [&NiagaraSpriteRendererProperties]
  renderers [{rendererClass=&NiagaraSpriteRendererProperties}]
 }
 Rico_bullets003 {
  bEnabled=false
  emitterUpdateScript [
   &EmitterState()
   &SpawnBurst_Instantaneous(Spawn Count=18, Use Spawn Probability=false)
  ]
  particleSpawnScript [
   &InitializeParticle(
    Color=&LerpLinearColors(StartColor=<rgba 2 0.575 0.1 1>, EndColor=<rgba 5 3.1239 0.7531 1>, LerpFactor=$System.Impactintencity)
    Lifetime Max=0.25
    Lifetime Min=0.15
    Lifetime Mode=::NewEnumerator1
    Sprite Size=&Lerp_Vector2(
     A=&RandomRangeVector2D(Minimum=<3 5>, Maximum=<2 10>)
     B=&RandomRangeVector2D(Minimum=<2 15>, Maximum=<4 10>)
     Alpha=$System.Impactintencity
    )
    Sprite Size Mode=::NewEnumerator4
    UsePositionOffset=false
    Write Material Random=true
   ) {module=InitializeParticle001}
   &AddVelocityInCone(
    Cone Angle=30
    Cone Axis=&RotateVector(
     Vector To Rotate:Vector3f=<0.35 0 1>
     Yaw:NiagaraFloat=0
     Pitch:NiagaraFloat=0
     Roll:NiagaraFloat=&Multiply_Float(A=35, B=&RandomRangeFloat(Minimum=-1, Maximum=1))
     Rotation Rate:NiagaraFloat=1
    )
    Module.Use Velocity Falloff On Cone Axis=false
    Velocity Strength=&Multiply_Float(A=&UniformRangedFloat(Minimum=300, Maximum=200), B=10)
   )
  ]
  particleUpdateScript [
   &ParticleState()
   &ScaleColor(
    Scale Alpha=&FloatFromCurve(FloatCurve=DataInterface() {propertyValues {Curve {keys [
         {tangentMode=RCTM_User, time=0, value=1}
         {arriveTangent=-0.0405, interpMode=RCIM_Cubic, leaveTangent=-0.0405, tangentMode=RCTM_User, time=0.7009, value=1}
         {arriveTangent=-8.0089, interpMode=RCIM_Cubic, leaveTangent=-8.0089, tangentMode=RCTM_User, time=1, value=0}
        ]}}})
    ScaleA=true
    ScaleRGB=true
   )
   &Drag(
    Drag=&Lerp_Float(
     A=&Multiply_Float(A=&RandomRangeFloat(Minimum=10, Maximum=12), B=2.2)
     B=&Multiply_Float(A=&RandomRangeFloat(Minimum=10, Maximum=12), B=2.2)
     Alpha=0.5
    )
    Use Linear Drag=true
    Use Rotational Drag=false
   )
   &SolveForcesAndVelocity(Clamp Velocity=false, Limit Acceleration=false)
   &ScaleSpriteSize(Scale Factor=&MakeVector2D(
     X=&FloatFromCurve(FloatCurve=DataInterface() {propertyValues {Curve {keys [{arriveTangent=-0.0405, interpMode=RCIM_Cubic, time=0, value=1}, {interpMode=RCIM_Cubic, leaveTangent=-8.0089, time=1, value=0.75}]}}})
     Y=&FloatFromCurve(FloatCurve=DataInterface() {propertyValues {Curve {keys [{arriveTangent=-2.7038, interpMode=RCIM_Cubic, time=0, value=1}, {interpMode=RCIM_Cubic, leaveTangent=-8.0089, time=1, value=0}]}}})
    ))
   &DynamicMaterialParameters(
    Index 0 Param 1=1
    Index 0 Param 2=1
    Index 0 Param 3=&FloatFromCurve(FloatCurve=DataInterface() {propertyValues {Curve {keys [{interpMode=RCIM_Cubic, tangentMode=RCTM_User, time=0, value=1}, {interpMode=RCIM_Cubic, tangentMode=RCTM_User, time=1, value=0}]}}})
    Index 0 Param 4=1
    Module.Param0WriteEnabled=true
    Module.Param1WriteEnabled=true
    Module.Param2WriteEnabled=true
    Module.Param3WriteEnabled=true
   ) {module=DynamicMaterialParameters001}
   &ScaleSpriteSizeBySpeed(Max Scale Factor=<0.8 6>, Min Scale Factor=<1 3>, Velocity Threshold=2000)
  ]
  rendererClasses [&NiagaraSpriteRendererProperties]
  renderers [{rendererClass=&NiagaraSpriteRendererProperties}]
 }
 Small_rico {
  emitterUpdateScript [
   &EmitterState()
   &SpawnBurst_Instantaneous(Spawn Count=18, Use Spawn Probability=false)
  ]
  particleSpawnScript [
   &InitializeParticle(
    Color=&LerpLinearColors(StartColor=<rgba 2 0.575 0.1 1>, EndColor=<rgba 5 3.1239 0.7531 1>, LerpFactor=$System.Impactintencity)
    Lifetime=&Lerp_Float(
     A=0.05
     B=0.15
     Alpha=&FloatFromCurve(
      FloatCurve=DataInterface() {propertyValues {Curve {keys [<at 0 0 RCIM_Cubic>, {arriveTangent=3.573, interpMode=RCIM_Cubic, leaveTangent=3.573, tangentMode=RCTM_User, time=1, value=1}]}}}
      CurveIndex=&RandomRangeFloat(Minimum=0, Maximum=1)
     )
    )
    Mass Max=3
    Mass Mode=::NewEnumerator3
    Sprite Size Mode=::NewEnumerator3
    Uniform Sprite Size=2
    UsePositionOffset=false
    Write Material Random=true
   ) {module=InitializeParticle001}
   &AddVelocityInCone(
    Cone Angle=10
    Cone Axis=&RotateVector(
     Vector To Rotate:Vector3f=<0.35 0 1>
     Yaw:NiagaraFloat=0
     Pitch:NiagaraFloat=0
     Roll:NiagaraFloat=&Multiply_Float(A=22, B=&RandomRangeFloat(Minimum=-1, Maximum=1))
     Rotation Rate:NiagaraFloat=1
    )
    Module.Use Velocity Falloff On Cone Axis=false
    Velocity Strength=&Multiply_Float(A=&UniformRangedFloat(Minimum=300, Maximum=200), B=12)
   )
  ]
  particleUpdateScript [
   &ParticleState()
   &ScaleColor(
    Scale Alpha=&FloatFromCurve(FloatCurve=DataInterface() {propertyValues {Curve {keys [
         {tangentMode=RCTM_User, time=0, value=1}
         {arriveTangent=-0.0405, interpMode=RCIM_Cubic, leaveTangent=-0.0405, tangentMode=RCTM_User, time=0.7009, value=1}
         {arriveTangent=-8.0089, interpMode=RCIM_Cubic, leaveTangent=-8.0089, tangentMode=RCTM_User, time=1, value=0}
        ]}}})
    ScaleA=true
    ScaleRGB=true
   )
   &Drag(Drag=10, Use Linear Drag=true, Use Rotational Drag=true)
   &GravityForce(Gravity=<0 0 -9800>)
   &SolveForcesAndVelocity(Clamp Velocity=false, Limit Acceleration=false)
   &ScaleSpriteSizeBySpeed(Max Scale Factor=&RandomRangeVector2D(Minimum=<0.5 10>, Maximum=<0.5 20>), Min Scale Factor=<0.1 0.5>)
   &DynamicMaterialParameters(
    Index 0 Param 1=1
    Index 0 Param 2=1
    Index 0 Param 3=&FloatFromCurve(FloatCurve=DataInterface() {propertyValues {Curve {keys [<at 0 1 RCIM_Cubic>,<at 1 0 RCIM_Cubic>]}}})
    Index 0 Param 4=1
    Module.Index0Param0Write=true
    Module.Index0Param1Write=true
    Module.Index0Param2Write=true
    Module.Index0Param3Write=true
   )
   &Color(Color=&MultiplyLinearColor(
     Linear Color=&ColorFromCurve(ColorCurve=DataInterface() {propertyValues {
        AlphaCurve {keys [<at 0.8243 1 RCIM_Cubic>,<at 1 0 RCIM_Cubic>]}
        BlueCurve {keys [
          <at 0.0115 1.25 RCIM_Cubic>
          {arriveTangentWeight=0.209, interpMode=RCIM_Cubic, leaveTangentWeight=0.1865, tangentMode=RCTM_User, time=0.1937, value=0.2}
          {arriveTangent=-0.2642, interpMode=RCIM_Cubic, leaveTangent=-0.2642, time=0.6501, value=0}
          <at 0.9508 0 RCIM_Cubic>
         ]}
        GreenCurve {keys [
          <at 0.0115 1.25 RCIM_Cubic>
          {
           arriveTangent=-1.5659
           arriveTangentWeight=0.1735
           interpMode=RCIM_Cubic
           leaveTangent=-1.5659
           leaveTangentWeight=0.1548
           time=0.1937
           value=0.6
          }
          {arriveTangent=-0.7925, interpMode=RCIM_Cubic, leaveTangent=-0.7925, time=0.6501, value=0.25}
          <at 0.9508 0 RCIM_Cubic>
         ]}
        RedCurve {keys [
          <at 0.0115 1.25 RCIM_Cubic>
          {
           arriveTangent=-0.3915
           arriveTangentWeight=0.1125
           interpMode=RCIM_Cubic
           leaveTangent=-0.3915
           leaveTangentWeight=0.1004
           time=0.1937
           value=1
          }
          <at 0.6501 1 RCIM_Cubic>
          <at 0.9508 1 RCIM_Cubic>
         ]}
       }})
     Scale Factor=&MakeFloat4FromVectorAndFloat(
      "Vector (RGB)"=&VectorFromFloat(Value=&RandomRangeFloat(Minimum=0.5, Maximum=2))
      "Float (Alpha)"=1
     )
    ))
  ]
  rendererClasses [&NiagaraSpriteRendererProperties]
  renderers [{rendererClass=&NiagaraSpriteRendererProperties}]
 }
}
summary {
 systemName=NS_BallisticSheild_Impact
 emitters [
  {emitterName=Rico_bullets, rendererClasses [&NiagaraSpriteRendererProperties]}
  {emitterName=Small_rico, rendererClasses [&NiagaraSpriteRendererProperties]}
  {bEnabled=false, emitterName=Rico_bullets002, rendererClasses [&NiagaraSpriteRendererProperties]}
  {bEnabled=false, emitterName=Rico_bullets003, rendererClasses [&NiagaraSpriteRendererProperties]}
 ]
 userVariables [User.ImpactStage {type=NiagaraFloat, defaultValue=0, description=""}]
}
