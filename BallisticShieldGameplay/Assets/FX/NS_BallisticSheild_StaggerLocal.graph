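# Niagara system for the ballistic shield's local stagger effect, a brief burst of two mesh emitters at a shared offset from the origin.
# The Empty emitter's mesh expands while its very high alpha fades out over a quarter second.
# Wind_Cone's stretched, Y-oriented mesh grows from zero as its alpha rises then falls.
# Curves drive scale and alpha; the CameraEffectLifteime and ForwardDot user parameters are unused by the listed modules.
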NiagaraSystem /BallisticShieldGameplay/Assets/FX/NS_BallisticSheild_StaggerLocal
dependencies {usedDataInterfaces [&NiagaraDataInterfaceCurve], usedDynamicInputs [], usedModules [&SystemState]}
emitters {
 Empty {
  emitterUpdateScript [
   &EmitterState()
   &SpawnBurst_Instantaneous(Use Spawn Probability=false, UseLoopCountLimit=false)
  ]
  particleSpawnScript [&InitializeParticle(
    Color=<rgba 1 1 1 4749.8232>
    Lifetime=0.25
    Mesh Scale Mode=::NewEnumerator3
    Position Offset=<0 50 100>
    UsePositionOffset=true
    Write Material Random=true
   )]
  particleUpdateScript [
   &ParticleState()
   &ScaleMeshSize(Scale Factor=&VectorFromFloat(Value=&FloatFromCurve(FloatCurve=DataInterface() {propertyValues {Curve {keys [
          {arriveTangent=1.5387, interpMode=RCIM_Cubic, leaveTangent=1.5387, tangentMode=RCTM_User, time=0, value=1}
          {arriveTangent=0.1275, interpMode=RCIM_Cubic, leaveTangent=0.1275, tangentMode=RCTM_User, time=1, value=2.5}
         ]}}})))
   &ScaleColor(
    Scale Alpha=&FloatFromCurve(FloatCurve=DataInterface() {propertyValues {Curve {keys [{arriveTangent=-3.0155, interpMode=RCIM_Cubic, leaveTangent=-3.0155, tangentMode=RCTM_User, time=0, value=1}, <at 1 0>]}}})
    ScaleA=true
    ScaleRGB=true
   )
  ]
  rendererClasses [&NiagaraMeshRendererProperties]
  renderers [{rendererClass=&NiagaraMeshRendererProperties, rendererIndex=0}]
 }
 Wind_Cone {
  emitterUpdateScript [
   &EmitterState()
   &SpawnBurst_Instantaneous(Use Spawn Probability=false, UseLoopCountLimit=false)
  ]
  particleSpawnScript [
   &InitializeParticle(
    Color=<rgba 1 1 1 250>
    Lifetime=0.25
    Mesh Scale=<0.5 1.5 1.5>
    Mesh Scale Mode=::NewEnumerator4
    Position Offset=<0 50 100>
    UsePositionOffset=true
    Write Material Random=true
   )
   &InitialMeshOrientation(Mesh Orientation Mode=::NewEnumerator2, Orientation Vector=<0 1 0>, Use Rotation Vector=true)
  ]
  particleUpdateScript [
   &ParticleState()
   &ScaleMeshSize(Scale Factor=&VectorFromFloat(Value=&FloatFromCurve(FloatCurve=DataInterface() {propertyValues {Curve {keys [{tangentMode=RCTM_User, time=0, value=0}, {tangentMode=RCTM_User, time=1, value=1}]}}})))
   &ScaleColor(
    Scale Alpha=&FloatFromCurve(FloatCurve=DataInterface() {propertyValues {Curve {keys [
         {interpMode=RCIM_Cubic, tangentMode=RCTM_User, time=0, value=0}
         {interpMode=RCIM_Cubic, tangentMode=RCTM_User, time=0.5, value=1}
         {tangentMode=RCTM_User, time=1, value=0}
        ]}}})
    Scale RGB=&VectorFromFloat(Value=1)
    ScaleA=true
    ScaleRGB=true
   )
  ]
  rendererClasses [&NiagaraMeshRendererProperties]
  renderers [{rendererClass=&NiagaraMeshRendererProperties, rendererIndex=0}]
 }
}
summary {
 systemName=NS_BallisticSheild_StaggerLocal
 emitters [
  {emitterName=Wind_Cone, rendererClasses [&NiagaraMeshRendererProperties]}
  {emitterName=Empty, rendererClasses [&NiagaraMeshRendererProperties]}
 ]
 userVariables [
  User.CameraEffectLifteime {type=NiagaraFloat, defaultValue=1.3, description=""}
  User.ForwardDot {type=NiagaraFloat, defaultValue=1, description=""}
 ]
}
